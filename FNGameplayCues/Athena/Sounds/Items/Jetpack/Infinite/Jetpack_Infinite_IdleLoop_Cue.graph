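# Sound cue for the infinite jetpack's idle loop, classed as a gadget item sound and spatialized with the jetpack attenuation.
# A quality-level node picks the standard looping wave for its first slot and a shared "opt" looping wave for the other three.
# An enveloper at the root eases the volume down to half over the first ten seconds, on top of a low overall volume multiplier.
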SoundCue /FNGameplayCues/Athena/Sounds/Items/Jetpack/Infinite/Jetpack_Infinite_IdleLoop_Cue
root=Enveloper_0
nodes [
 {
  id=Enveloper_0
  children [QualityLevel_0]
  properties {pitchCurve {editorCurveData {keys [{time=0}]}}, volumeCurve {editorCurveData {keys [<at 0 1>,<at 10 0.5>]}}}
 }
 {id=QualityLevel_0, children [WavePlayer_0,WavePlayer_1,WavePlayer_1,WavePlayer_1]}
 {id=WavePlayer_0, properties {bLooping=true, soundWaveAssetPtr=/FNGameplayCues/Athena/Sounds/Items/Jetpack/Jetpack_IdleLoop_01}}
 {id=WavePlayer_1, properties {bLooping=true, soundWaveAssetPtr=/FNGameplayCues/Athena/Sounds/Items/Jetpack/Jetpack_IdleLoop_01_opt}}
]
properties {
 attenuationSettings=/Game/Sounds/Attenuation/Fort_Jetpack_Att
 maxDistance=13067
 soundClassObject=&Fort_Gadgets_Items
 volumeMultiplier=0.17
}
