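# Sound cue for the infinite jetpack's stop sound, an Athena item gameplay cue sound.
# A single wave player for the Jetpack_Stop_01 wave feeds a modulator that applies a slight random upward pitch variation.
# It plays at reduced volume through the Fort_Jetpack_Att attenuation in the gadget items sound class.
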SoundCue /FNGameplayCues/Athena/Sounds/Items/Jetpack/Infinite/Jetpack_Infinite_Stop_Cue
root=Modulator_0
nodes [
 {id=Modulator_0, children [WavePlayer_0], properties {pitchMax=1.1, pitchMin=1}}
 {id=WavePlayer_0, properties {soundWaveAssetPtr=/Game/Athena/Sounds/Items/Jetpack/Jetpack_Stop_01}}
]
properties {
 attenuationSettings=/Game/Sounds/Attenuation/Fort_Jetpack_Att
 maxDistance=13067
 soundClassObject=&Fort_Gadgets_Items
 volumeMultiplier=0.2
}
